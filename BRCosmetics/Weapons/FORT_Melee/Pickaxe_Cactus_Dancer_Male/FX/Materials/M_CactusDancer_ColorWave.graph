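# Translucent unlit Niagara mesh-particle material for the Cactus Dancer pickaxe, drawing glowing color-wave lines.
# Color samples the Cactus Rocker curve atlas through a sine-pulsed, U-centered sphere mask, boosts saturation, and tints by particle color.
# Opacity repeats lines from a distance-field texture, with particle dynamic parameters setting offset and count.
# Optional view-angle and UV gradient fades shape the lines; depth fade applies everywhere except mobile.
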Material /BRCosmetics/Weapons/FORT_Melee/Pickaxe_Cactus_Dancer_Male/FX/Materials/M_CactusDancer_ColorWave
graph {
 properties {
  bUsedWithNiagaraMeshParticles=true
  blendMode=BLEND_Translucent
  pixelDepthOffsetMode=PDOM_Legacy
  shadingModel=MSM_Unlit
  usedShadingModels=Unlit
 }
 nodes [
  n1 = ParticleColor()
  n2 = TextureCoordinate() {vTiling=0}
  n3 = Constant2Vector() {r=0.5}
  n4 = Sine(None=Time()) {period=3}
  n5 = LinearInterpolate(A=Constant() {r=0.4}, B=Constant() {r=0.6}, Alpha=Divide(A=Multiply(A=@n4) {constB=2}))
  n6 = Constant()
  n7 = CurveAtlasRowParameter(CurveTime=OneMinus(None=SphereMask(A=@n2, B=@n3, Radius=@n5, Hardness=@n6))) {
   atlas=/Game/Accessories/FORT_Backpacks/Backpack_F_MED_Cactus_Rocker/Textures/CA_CactusRocker
   curve=/Game/Accessories/FORT_Backpacks/Backpack_F_MED_Cactus_Rocker/Textures/CactusRockerColor
  }
  n8 = Desaturation(None=@n7, Fraction=Constant() {r=-0.5})
  n9 = ScalarParameter() {defaultValue=1, parameterName=UseCactusDancerColor}
  MaterialFunctionCall(Emissive=Multiply(A=LinearInterpolate(A=@n1.RGB, B=@n8, Alpha=@n9), B=@n1.RGB)) {materialFunction=&MF_FX_EmissiveAdjustments}
  n10 = TextureSampleParameter2D() {
   group=Lines
   parameterName=SDF
   samplerType=SAMPLERTYPE_LinearGrayscale
   texture=/BRCosmetics/Weapons/FORT_Melee/Pickaxe_Cactus_Dancer_Male/FX/Textures/T_CactusDancer_DistanceFields_M
  }
  n11 = DynamicParameter() {defaultValue=<rgba 0.34 3 1 1>, paramNames [Offset,LineAmount,Param3,Param4]}
  n12 = Add(A=@n10.R, B=AppendVector(A=Constant(), B=@n11.Offset))
  n13 = ComponentMask(None=Frac(None=Multiply(A=@n12, B=@n11.LineAmount))) {g=true}
  n14 = Multiply(
   A=ComponentMask(None=OneMinus(None=@n13)) {g=true}
   B=SmoothStep(
    Min=ScalarParameter() {group=Lines, parameterName=SmoothStepMin}
    Max=ScalarParameter() {defaultValue=0.3, group=Lines, parameterName=SmoothStepMax}
    Value=@n13
   )
  )
  n15 = Multiply(
   A=Multiply(A=Power(Base=@n14) {constExponent=8}) {constB=2}
   B=Power(Base=@n10.R, Exp=ScalarParameter() {defaultValue=2, group=Lines, parameterName=MaskPower})
  )
  n16 = OneMinus(None=@n10.R)
  n17 = Saturate(None=Multiply(A=Multiply(A=@n15, B=@n16), B=ScalarParameter() {defaultValue=12, group=Lines, parameterName=FinalStrenght}) {constB=12})
  n18 = ScalarParameter() {defaultValue=4, parameterName=AngleFadeExponent}
  n19 = LinearInterpolate(
   A=@n17
   B=Multiply(
    A=@n17
    B=MaterialFunctionCall("FadeSensitivity (optional)"=@n18) {materialFunction=&FadeBasedOnViewAngle}.Result
   )
   Alpha=ScalarParameter() {parameterName=AnglefadeAmount}
  )
  n20 = TextureCoordinate()
  n21 = OneMinus(None=@n20)
  n22 = ScalarParameter() {group=OpacityGradient, parameterName=InvertOpacityGradient}
  n23 = SmoothStep(
   Min=ScalarParameter() {group=OpacityGradient, parameterName=OpacityGradientSmothStepMin}
   Max=ScalarParameter() {defaultValue=1, group=OpacityGradient, parameterName=OpacityGradientSmothStepMax}
   Value=DotProduct(A=LinearInterpolate(A=@n21, B=@n20, Alpha=@n22), B=VectorParameter() {parameterName=OpacityGradientChannel, defaultValue=<rgba 0 1 0 1>}.RGB)
  )
  n24 = LinearInterpolate(A=@n19, B=Multiply(A=@n19, B=Saturate(None=@n23)), Alpha=ScalarParameter() {group=OpacityGradient, parameterName=UseOpacityGradient})
  n25 = Multiply(A=@n24, B=DepthFade() {fadeDistanceDefault=10})
  Multiply(A=@n1.A, B=ShadingPathSwitch(Default=@n25, Deferred=@n25, Mobile=@n24) {desc=Migrated from FeatureLevelSwitch})
 ]
}
