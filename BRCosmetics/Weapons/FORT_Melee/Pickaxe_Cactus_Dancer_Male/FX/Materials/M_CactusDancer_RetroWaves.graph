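# Unlit additive material for the Cactus Dancer pickaxe effects, drawing wavy, banded "retro waves" on Niagara sprite or mesh particles.
# Color comes from a curve atlas gradient across the bounding box, oversaturated by negative desaturation and tinted by particle color.
# Perlin noise, shifted by the RandomXOffset dynamic parameter, distorts repeating V-axis stripes and their edge fades to form the mask.
# Noise sampling uses explicit UV derivatives alongside Frac, and particle alpha scales the mask.
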Material /BRCosmetics/Weapons/FORT_Melee/Pickaxe_Cactus_Dancer_Male/FX/Materials/M_CactusDancer_RetroWaves
graph {
 properties {
  bUsedWithNiagaraMeshParticles=true
  bUsedWithNiagaraSprites=true
  blendMode=BLEND_Additive
  pixelDepthOffsetMode=PDOM_Legacy
  previewMesh=/BRCosmetics/Weapons/FORT_Melee/Pickaxe_Cube_Ninja_Male/FX/SM_Pickaxe_Cube_Ninja_Male_Impact_ShockWave_01
  shadingModel=MSM_Unlit
  usedShadingModels=Unlit
 }
 nodes [
  Constant()
  n1 = CurveAtlasRowParameter(CurveTime=Add(A=MaterialFunctionCall() {materialFunction=&BoundingBoxBased_0-1_UVW}.R) {constB=-0.15}) {
   atlas=/Game/Accessories/FORT_Backpacks/Backpack_F_MED_Cactus_Rocker/Textures/CA_CactusRocker
   curve=/Game/Accessories/FORT_Backpacks/Backpack_F_MED_Cactus_Rocker/Textures/CactusRockerColor
  }
  n2 = Constant() {r=-0.8}
  n3 = ParticleColor()
  MaterialFunctionCall(Emissive=Multiply(A=Desaturation(None=@n1, Fraction=@n2), B=@n3.RGB)) {materialFunction=&MF_FX_EmissiveAdjustments}
  n4 = DynamicParameter() {defaultValue=<rgba 1 0 1 1>, paramNames [Panning, RandomXOffset, Param3, Param4]}
  n5 = AppendVector(A=@n4.RandomXOffset, B=@n4.RandomXOffset)
  n6 = Add(A=TextureCoordinate() {uTiling=2, vTiling=0.25}, B=@n5)
  n7 = TextureCoordinate()
  n8 = Multiply(A=TextureSample!(UVs=Frac(None=@n6), Tex, "DDX(UVs)"=DDX(Value=@n7), "DDY(UVs)"=DDY(Value=@n7), Apply View MipBias) {mipValueMode=TMVM_Derivative, texture=/Game/Global/Textures/Default/Noise/3xPerlin_M}.R) {constB=1.5}
  n9 = Multiply(A=@n8) {constB=0.5}
  n10 = Add(A=TextureCoordinate(), B=@n9)
  n11 = SmoothStep(Min=Constant() {r=0.15}, Max=Constant() {r=0.4}, Value=Subtract(A=ComponentMask(None=@n10) {g=true}) {constA=0.2, constB=0.2})
  n12 = Add(A=TextureCoordinate(), B=@n9)
  n13 = Subtract(A=OneMinus(None=ComponentMask(None=@n12) {g=true})) {constB=0.1}
  n14 = Add(A=TextureCoordinate() {uTiling=5, vTiling=5}, B=Multiply(A=@n5) {constB=2})
  n15 = ComponentMask(None=Frac(None=Add(A=@n14, B=@n8))) {g=true}
  n16 = OneMinus(None=ComponentMask(None=@n15) {g=true})
  n17 = SmoothStep(Min=Constant(), Max=Constant() {r=0.3}, Value=@n15)
  n18 = Multiply(A=Multiply(A=Saturate(None=@n11), B=Saturate(None=@n13)), B=Power(Base=Multiply(A=@n16, B=@n17)))
  Multiply(A=Saturate(None=Multiply(A=@n18) {constB=2}), B=@n3.A)
 ]
}
