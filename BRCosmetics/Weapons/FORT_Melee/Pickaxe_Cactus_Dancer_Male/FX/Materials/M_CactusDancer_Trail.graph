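# Unlit, two-sided translucent Niagara ribbon material for the Cactus Dancer pickaxe trail.
# A sine wave, panned by a dynamic parameter and modulated by scrolling Perlin noise, wobbles procedural core and optional outline bands plus time-animated side layers.
# Color comes from the Cactus Rocker backpack's curve atlas, with saturation boosted by time of day, multiplied by particle color.
# Opacity combines ribbon-end fades, view angle, depth fade, particle alpha and optional alpha falloff.
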Material /BRCosmetics/Weapons/FORT_Melee/Pickaxe_Cactus_Dancer_Male/FX/Materials/M_CactusDancer_Trail
graph {
 properties {
  bUsedWithNiagaraRibbons=true
  blendMode=BLEND_Translucent
  pixelDepthOffsetMode=PDOM_Legacy
  shadingModel=MSM_Unlit
  twoSided=true
  usedShadingModels=Unlit
 }
 nodes [
  n1 = LinearInterpolate(Alpha=NamedRerouteUsage()) {constA=0.1, constB=0.05}
  n2 = Constant() {r=0.5}
  n3 = TextureCoordinate()
  n4 = TextureCoordinate() {coordinateIndex=1}
  n5 = Add(A=ComponentMask(None=Multiply(A=@n4) {constB=7}) {r=true}, B=DynamicParameter() {paramNames [Panning,Param2,Param3,Param4]}.Panning)
  n6 = TextureCoordinate() {coordinateIndex=1, vTiling=0.2}
  n7 = Add(
   A=ComponentMask(None=TextureCoordinate()) {r=true}
   B=TextureSample(UVs=Panner(Coordinate=@n6) {speedX=0.2, speedY=-0.1}) {texture=/Game/Global/Textures/Default/Noise/3xPerlin_M}.R
  )
  n8 = Multiply(A=Divide(A=Sine(None=@n5)), B=LinearInterpolate(Alpha=@n7) {constA=0.3, constB=0.1}) {constB=0.25}
  n9 = Add(A=@n3, B=@n8)
  n10 = Multiply(A=@n9) {constB=2}
  n11 = LinearInterpolate(
   B=MaterialFunctionCall!(Width=Multiply(A=@n1) {constB=0.5}, Sharpness=@n2, Offset, Direction Switch, Compare, Input Coordinates=@n10) {materialFunction=/Engine/Functions/Engine_MaterialFunctions03/Procedurals/GeneratedBand}.Result
   Alpha=ScalarParameter() {defaultValue=1, parameterName=UseOutline}
  ) {constB=0}
  n12 = MaterialFunctionCall(
   Width=Multiply(A=ScalarParameter() {defaultValue=1, parameterName=Width}, B=@n1)
   Sharpness=Multiply(A=@n2, B=ScalarParameter() {defaultValue=1, parameterName=Width2})
   Input Coordinates=@n9
  ) {materialFunction=&GeneratedBand}
  n13 = Add(A=@n11, B=@n12.Result)
  n14 = ComponentMask(None=Add(A=Multiply(A=@n8), B=Add(A=@n3) {constB=-0.5})) {g=true}
  n15 = Constant()
  n16 = SmoothStep(Min=@n15, Max=@n15, Value=@n14) {constMax=0.5}
  n17 = Distance(A=@n14, B=@n16)
  n18 = OneMinus(None=@n17)
  n19 = Power(Base=Multiply(A=@n18, B=OneMinus(None=@n16))) {constExponent=4}
  n20 = TextureCoordinate() {coordinateIndex=1}
  n21 = Add(A=ComponentMask(None=Multiply(A=@n20)) {g=true}, B=@n18)
  n22 = Time()
  n23 = Add(A=@n21, B=@n22)
  n24 = Power(Base=Multiply(A=@n17, B=@n16)) {constExponent=4}
  n25 = Add(A=OneMinus(None=@n21), B=@n22)
  n26 = Add(A=Multiply(A=@n19, B=Sine(None=@n23)), B=Multiply(A=@n24, B=Sine(None=@n25)))
  n27 = Subtract(
   A=Subtract(A=Saturate(None=@n26), B=@n12.Result)
   B=MaterialFunctionCall(Width=@n1, Sharpness=@n2, Offset=Constant() {r=0.5}, Compare=@n18, Input Coordinates=@n10) {materialFunction=&GeneratedBand}.Result
  )
  n28 = SphereMask(A=TextureCoordinate() {uTiling=0}, B=Constant2Vector() {g=0.5}, Radius=Constant() {r=0.5}, Hardness=Constant())
  n29 = Add(A=@n13, B=Saturate(None=Multiply(A=@n28, B=@n27)))
  n30 = Subtract(A=@n19) {constB=0.05}
  n31 = Subtract(A=@n24) {constB=0.05}
  n32 = Constant() {r=0.5}
  n33 = Constant() {r=0.9}
  n34 = TextureCoordinate()
  n35 = ComponentMask(None=OneMinus(None=@n34)) {r=true}
  n36 = Multiply(A=Multiply(A=Add(A=@n30, B=@n31), B=SmoothStep(Min=@n32, Max=@n33, Value=@n35)), B=@n28)
  n37 = SmoothStep(Min=Constant(), Max=Constant() {r=0.5}, Value=@n35)
  n38 = SmoothStep(Min=Constant(), Max=Constant() {r=0.1}, Value=ComponentMask(None=@n34) {r=true})
  n39 = Saturate(None=Multiply(A=Add(A=@n29, B=@n36), B=Multiply(A=@n37, B=@n38)))
  n40 = MaterialFunctionCall("FadeSensitivity (optional)"=Constant() {r=1.25}) {materialFunction=&FadeBasedOnViewAngle}
  n41 = ParticleColor()
  n42 = Multiply(A=Multiply(A=Multiply(A=@n39, B=@n40.Result), B=DepthFade() {fadeDistanceDefault=25}), B=@n41.A)
  n43 = OneMinus(None=ComponentMask(None=TextureCoordinate()) {r=true})
  n44 = ScalarParameter() {defaultValue=1, parameterName=AlphaFade_Power}
  LinearInterpolate(A=@n42, B=Multiply(A=@n42, B=Power(Base=@n43, Exp=@n44) {constExponent=5}), Alpha=ScalarParameter() {parameterName=UseAlphaFade})
  n45 = CurveAtlasRowParameter(CurveTime=@n28) {
   atlas=/Game/Accessories/FORT_Backpacks/Backpack_F_MED_Cactus_Rocker/Textures/CA_CactusRocker
   curve=/Game/Accessories/FORT_Backpacks/Backpack_F_MED_Cactus_Rocker/Textures/CactusRockerColor
  }
  n46 = CurveAtlasRowParameter(CurveTime=Add(A=Multiply(A=@n11) {constB=0.5}, B=@n12.Result)) {
   atlas=&CA_CactusRocker
   curve=&CactusRockerColor
  }
  n47 = LinearInterpolate(A=@n45, B=@n46, Alpha=@n13)
  n48 = Constant() {r=-0.5}
  n49 = MaterialFunctionCall(Value_Morning=@n48, Value_Day=@n48, Value_Evening=@n48, Value_Night=Constant() {r=-0.25}) {materialFunction=&Scalar_Time_of_Day}
  MaterialFunctionCall(Emissive=Multiply(A=@n41.RGB, B=Desaturation(None=@n47, Fraction=@n49.Result))) {materialFunction=&MF_FX_EmissiveAdjustments}
  Multiply(A=@n1)
  NamedRerouteDeclaration(None=Multiply(A=@n7, B=ScalarParameter() {defaultValue=1, parameterName=WobbleIntensity})) {name=WaveDistort, nodeColor=<rgba 1 0.4219 0 1>}
 ]
}
