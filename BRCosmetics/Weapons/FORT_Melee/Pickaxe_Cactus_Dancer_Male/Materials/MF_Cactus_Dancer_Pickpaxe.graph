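# Material function for the Cactus Dancer Male pickaxe that adds an audio-reactive spectrum-bar glow, behind a static switch.
# MF_AudioReactivity drives six SmoothCeil band masks, which stepped thresholds combine into a bar mask.
# An FX texture supplies band IDs, glow and gradient masks, alongside a sine-waved pre-skinned position gradient and Fresnel.
# Cactus Rocker curve-atlas colors feed base color and time-of-day-adjusted emissive, set with metallic, specular and roughness.
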MaterialFunction /BRCosmetics/Weapons/FORT_Melee/Pickaxe_Cactus_Dancer_Male/Materials/MF_Cactus_Dancer_Pickpaxe
graph {nodes [
  n1 = Divide(A=VertexInterpolator(VS=PreSkinnedPosition()).PS, B=PreSkinnedLocalBounds().Half Extents)
  n2 = ComponentMask(None=VectorParameter() {group=Cactus Dancer Pickaxe, parameterName=Offset, defaultValue=<rgba 0 0.1 1.2 0>}.RGB) {b=true, g=true, r=true}
  n3 = Rotator(
   Coordinate=ComponentMask(None=Add(A=@n1, B=@n2)) {b=true, g=true}
   Time=ScalarParameter() {defaultValue=0.6, group=Cactus Dancer Pickaxe, parameterName=Gradient Rotation}
  ) {speed=3.1415}
  Abs(None=ComponentMask(None=@n3) {desc=Vertical Pos, r=true})
  n4 = SetMaterialAttributes!(
   MaterialAttributes=NamedRerouteDeclaration(None=FunctionInput() {inputType=FunctionInput_MaterialAttributes}) {name=MA, nodeColor=<rgba 0.6172 0 1 1>}
   BaseColor=NamedRerouteUsage()
   EmissiveColor=NamedRerouteUsage()
   Metallic=ScalarParameter() {defaultValue=0.5, group=Cactus Dancer Pickaxe, parameterName=Metallic}
   Specular=ScalarParameter() {defaultValue=1, group=Cactus Dancer Pickaxe, parameterName=Specular}
   Roughness=ScalarParameter() {defaultValue=0.1, group=Cactus Dancer Pickaxe, parameterName=Roughness}
  ) {attributeSetTypes [
    #1
    #2
    #3
    #4
    #5
   ]}
  FunctionOutput(None=StaticSwitchParameter(True=@n4, False=NamedRerouteUsage()) {parameterName=UseCactusDancerMalePickaxeFX})
  n5 = OneMinus(None=Saturate(None=Abs(None=@n3)))
  n6 = Add(A=Time(), B=NamedRerouteDeclaration(None=ComponentMask(None=@n5) {g=true}) {name=HorizontalPos, nodeColor=<rgba 0 1 0.0156 1>})
  n7 = ScalarParameter() {defaultValue=0.5, group=Cactus Dancer Pickaxe, parameterName=WaveWidth}
  n8 = Multiply(A=Sine(None=Divide(A=@n6, B=@n7)), B=ScalarParameter() {defaultValue=0.025, group=Cactus Dancer Pickaxe, parameterName=WaveHeight})
  n9 = NamedRerouteUsage()
  NamedRerouteDeclaration(None=Saturate(None=Add(A=@n8, B=@n9))) {name=Modulated_ZPos, nodeColor=<rgba 0 0.2266 1 1>}
  n10 = TextureSample() {samplerType=SAMPLERTYPE_Masks, texture=/BRCosmetics/Weapons/FORT_Melee/Pickaxe_Cactus_Dancer_Male/Textures/T_Pickaxe_Cactus_Dancer_Male_FX}
  n11 = OneMinus(None=Fresnel() {baseReflectFraction=0, exponent=1})
  n12 = Add(A=@n10.R) {constB=-0.7}
  NamedRerouteDeclaration(None=LinearInterpolate(A=@n10.G, B=Power(Base=@n11), Alpha=Ceil(None=@n12) {desc=TrunkMask})) {name=Band_GlowMask, nodeColor=<rgba 0.1382 1 0 1>}
  n13 = NamedRerouteUsage()
  n14 = Abs(None=ConstantBiasScale(None=@n13) {bias=-0.5, scale=2})
  NamedRerouteDeclaration(None=OneMinus(None=StaticSwitchParameter(True=@n14, False=@n13) {group=Cactus Dancer Pickaxe, parameterName=Use_Center_Based_Spectrum})) {name=Spectrum_Pos, nodeColor=<rgba 0 1 0.8359 1>}
  n15 = MaterialFunctionCall() {materialFunction=/Game/Packages/Fortress_SharedMaterials/Base_Material_Functions/BaseCharacterShaderFunctions/Audio/MF_AudioReactivity}
  n16 = Multiply(A=NamedRerouteDeclaration(None=@n15.Amplitude) {name=Audio_Amplitude, nodeColor=<rgba 0 0.0879 1 1>}, B=@n15.Spectrum)
  n17 = ComponentMask(None=@n16) {r=true}
  n18 = Subtract(A=NamedRerouteUsage(), B=@n17)
  n19 = Constant() {r=0.25}
  NamedRerouteDeclaration(None=Saturate(None=MaterialFunctionCall!(In=@n18, Ramp Width=@n19) {materialFunction=/Engine/Functions/Engine_MaterialFunctions01/Math/SmoothCeil}.Result) {desc=Band 1}) {name=Band_1, nodeColor=<rgba 1 0 0.4922 1>}
  n20 = ComponentMask(None=@n16) {g=true}
  n21 = Add(A=@n17, B=@n20)
  n22 = Subtract(A=NamedRerouteUsage(), B=Power(Base=Multiply(A=@n21) {constB=0.5}) {constExponent=0.7})
  n23 = Constant() {r=0.25}
  NamedRerouteDeclaration(None=Saturate(None=MaterialFunctionCall(In=@n22, Ramp Width=@n23) {materialFunction=&SmoothCeil}.Result) {desc="Band  2"}) {name=Band_2, nodeColor=<rgba 1 0 0.4922 1>}
  n24 = Subtract(A=NamedRerouteUsage(), B=@n20)
  n25 = Constant() {r=0.25}
  NamedRerouteDeclaration(None=Saturate(None=MaterialFunctionCall(In=@n24, Ramp Width=@n25) {materialFunction=&SmoothCeil}.Result) {desc=Band 3}) {name=Band_3, nodeColor=<rgba 1 0 0.4922 1>}
  n26 = ComponentMask(None=@n16) {b=true}
  n27 = Add(A=@n20, B=@n26)
  n28 = Subtract(A=NamedRerouteUsage(), B=Power(Base=Multiply(A=@n27) {constB=0.5}) {constExponent=0.7})
  n29 = Constant() {r=0.25}
  NamedRerouteDeclaration(None=Saturate(None=MaterialFunctionCall(In=@n28, Ramp Width=@n29) {materialFunction=&SmoothCeil}.Result) {desc="Band  4"}) {name=Band_4, nodeColor=<rgba 1 0 0.4922 1>}
  n30 = Subtract(A=NamedRerouteUsage(), B=@n26)
  n31 = Constant() {r=0.25}
  NamedRerouteDeclaration(None=Saturate(None=MaterialFunctionCall(In=@n30, Ramp Width=@n31) {materialFunction=&SmoothCeil}.Result) {desc=Band 5}) {name=Band_5, nodeColor=<rgba 1 0 0.4922 1>}
  n32 = ComponentMask(None=@n16) {a=true}
  n33 = Subtract(A=NamedRerouteUsage(), B=@n32)
  n34 = Constant() {r=0.25}
  NamedRerouteDeclaration(None=Saturate(None=MaterialFunctionCall(In=@n33, Ramp Width=@n34) {materialFunction=&SmoothCeil}.Result) {desc=Band 6}) {name=Band_6, nodeColor=<rgba 1 0 0.4922 1>}
  n35 = Ceil(None=Subtract(A=NamedRerouteUsage()) {constB=0.15})
  n36 = Ceil(None=Subtract(A=NamedRerouteUsage()) {constB=0.3})
  n37 = OneMinus(None=@n36)
  n38 = Add(A=Multiply(A=OneMinus(None=@n35), B=NamedRerouteUsage()), B=Multiply(A=Multiply(A=@n35, B=@n37), B=NamedRerouteUsage()))
  n39 = Ceil(None=Subtract(A=NamedRerouteUsage()) {constB=0.45})
  n40 = Multiply(A=@n36, B=OneMinus(None=@n39))
  n41 = NamedRerouteUsage()
  n42 = Ceil(None=Subtract(A=NamedRerouteUsage()) {constB=0.55})
  n43 = Multiply(A=@n39, B=OneMinus(None=@n42))
  n44 = NamedRerouteUsage()
  n45 = Ceil(None=Subtract(A=NamedRerouteUsage()) {constB=0.9})
  n46 = Multiply(A=@n42, B=OneMinus(None=@n45))
  n47 = NamedRerouteUsage()
  n48 = NamedRerouteUsage()
  n49 = Add(A=Add(A=Multiply(A=@n40, B=@n41), B=Multiply(A=@n43, B=@n44)), B=Add(A=Multiply(A=@n46, B=@n47), B=Multiply(A=@n45, B=@n48)))
  n50 = Multiply(A=Ceil(None=NamedRerouteUsage()), B=Multiply(A=Add(A=@n38, B=@n49), B=NamedRerouteUsage()))
  n51 = Ceil(None=NamedRerouteUsage())
  NamedRerouteDeclaration(None=Saturate(None=Multiply(A=@n50, B=@n51))) {name=Spectrum_Bars_Mask, nodeColor=<rgba 1 0.2109 0 1>}
  NamedRerouteDeclaration(None=@n10.B) {name=Gradient, nodeColor=<rgba 1 0 0.5391 1>}
  n52 = Saturate(None=NamedRerouteUsage())
  NamedRerouteDeclaration(None=Fresnel(BaseReflectFractionIn=LinearInterpolate(Alpha=@n52) {constA=0.35, constB=0.175}) {baseReflectFraction=0.2, exponent=1.5}) {name=Edge_Falloff_Mask, nodeColor=<rgba 0.3359 0 1 1>}
  n53 = NamedRerouteUsage()
  n54 = CurveAtlasRowParameter(CurveTime=NamedRerouteUsage()) {
   atlas=/Game/Accessories/FORT_Backpacks/Backpack_F_MED_Cactus_Rocker/Textures/CA_CactusRocker
   curve=/Game/Accessories/FORT_Backpacks/Backpack_F_MED_Cactus_Rocker/Textures/CactusRockerColor
   group=Cactus Dancer Pickaxe
   parameterName=CA_Parameter
  }
  n55 = Desaturation(None=@n54, Fraction=Constant() {r=-1})
  n56 = NamedRerouteUsage()
  n57 = Multiply(A=@n53, B=LinearInterpolate(A=@n54, B=@n55, Alpha=@n56))
  n58 = Multiply(A=@n55, B=@n56)
  n59 = Multiply(A=@n58, B=ScalarParameter() {defaultValue=3, group=Cactus Dancer Pickaxe, parameterName=Spectrum_Emissive_Intensity})
  n60 = NamedRerouteUsage()
  NamedRerouteDeclaration(None=MaterialFunctionCall!(Color=LinearInterpolate(A=@n57, B=@n59, Alpha=@n60)) {materialFunction=/Game/Packages/Fortress_SharedMaterials/Anime_Master_Material/MF_TODColorCurve}.Result) {name=Emissive_Out, nodeColor=<rgba 1 0 0.1172 1>}
  NamedRerouteDeclaration(None=Multiply(A=Multiply(A=@n54, B=@n53), B=LinearInterpolate(A=@n57, B=@n58, Alpha=@n60))) {name=BaseColor_Out, nodeColor=<rgba 0.2188 1 0 1>}
  NamedRerouteDeclaration(None=@n10.R) {name=Spectrum_Band_IDs, nodeColor=<rgba 1 0 0.1406 1>}
  n61 = NamedRerouteUsage()
  n62 = Power(Base=Multiply(A=Add(A=@n26, B=@n32)) {constB=0.5}) {constExponent=0.7}
  Saturate(None=MaterialFunctionCall(In=Subtract(A=@n61, B=@n62), Ramp Width=Constant() {r=0.25}) {materialFunction=&SmoothCeil}.Result) {desc="Band  6"}
 ]}
